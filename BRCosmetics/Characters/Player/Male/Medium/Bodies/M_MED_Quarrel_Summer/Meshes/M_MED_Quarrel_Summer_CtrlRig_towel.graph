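# Control Rig for the Quarrel Summer male body that applies pose-driven corrections to the towel bones around the neck and shoulders.
# Head rotation relative to spine_05 is remapped to clamped weights that blend offset, rescaled local transforms onto back towel bones.
# A right-arm-driven additive correction and a zero-weight left towel override sit in a chain unconnected to BeginExecution.
# Left-arm remap and several right-towel offset calculations feed nothing.
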ControlRigBlueprint /BRCosmetics/Characters/Player/Male/Medium/Bodies/M_MED_Quarrel_Summer/Meshes/M_MED_Quarrel_Summer_CtrlRig_towel
graphs [
 RigVMFunctionLibrary {links [], nodes []}
 RigVMModel {nodes [
   n1 = &RigVMFunction_MathQuaternionToEuler(RotationOrder=ZYX, Value=&RigUnit_GetRelativeBoneTransform(Bone=upperarm_r, Space=spine_05).Transform.Rotation)
   /Script/ControlRig.RigUnit_ModifyBoneTransforms(
    BoneToModify="((Bone=\"dyn_towel_r_00\",Transform=(Rotation=(X=0,Y=0,Z=0,W=1),Translation=(X=0,Y=0,Z=-0.5),Scale3D=(X=1,Y=0.5,Z=0.5))),(Bone=\"dyn_towel_r_01\",Transform=(Rotation=(X=0,Y=0,Z=0,W=1),Translation=(X=0,Y=0,Z=1.5),Scale3D=(X=1,Y=1.6,Z=1.6))),(Bone=\"dyn_towel_r_0 ...(truncated)"
    ExecutePin=()
    Mode=AdditiveLocal
    WeightMaximum=1
    Weight=&RigVMFunction_MathFloatRemap(
     SourceMaximum=123
     SourceMinimum=30
     TargetMinimum=0
     bClamp=true
     Value=/Script/RigVM.RigVMFunction_MathFloatAbs(Value=@n1.Result.Z) {is=struct RigVMUnitNode, template="Absolute::Execute(in Value,out Result)"}.Result
    ).Result
   ) {is=struct RigVMUnitNode}
   n2 = /Script/ControlRig.RigUnit_GetInitialBoneTransform(Bone=dyn_towel_l_MSTR, Space=LocalSpace) {is=struct RigVMUnitNode}
   n3 = &RigVMFunction_MathQuaternionToEuler(RotationOrder=ZYX, Value=@n2.Transform.Rotation)
   > &RigUnit_ModifyBoneTransforms(
    BoneToModify=((Bone="dyn_towel_l_MSTR",Transform=(Rotation=(X=0,Y=0,Z=0,W=1),Translation=(X=0,Y=0,Z=0),Scale3D=(X=1,Y=1,Z=1))))
    Mode=OverrideLocal
    Weight=0
    WeightMaximum=1
    BoneToModify.0.Transform.Rotation=&RigVMFunction_MathQuaternionFromEuler(Euler=/Script/RigVM.RigVMFunction_MathVectorAdd(B=(X=-15,Y=0,Z=0), A=@n3.Result) {is=struct RigVMUnitNode, template="Add::Execute(in A,in B,out Result)"}.Result).Result
    BoneToModify.0.Transform.Translation=&RigVMFunction_MathVectorAdd(B=(X=0,Y=3.5,Z=6), A=@n2.Transform.Translation) {template="Add::Execute(in A,in B,out Result)"}.Result
   )
   RigUnit_BeginExecution = &RigUnit_BeginExecution()
   n4 = &RigUnit_GetRelativeBoneTransform(Bone=head, Space=spine_05)
   n5 = &RigUnit_GetInitialBoneTransform(Bone=dyn_towel_bk, Space=LocalSpace)
   n6 = &RigVMFunction_MathQuaternionToEuler(RotationOrder=ZYX, Value=@n5.Transform.Rotation)
   &RigUnit_ModifyBoneTransforms(
    BoneToModify=((Bone="dyn_towel_bk",Transform=(Rotation=(X=0,Y=0,Z=0,W=1),Translation=(X=0,Y=0,Z=0),Scale3D=(X=1.6,Y=1,Z=0.8))))
    Mode=OverrideLocal
    WeightMaximum=1
    Weight=&RigVMFunction_MathFloatRemap(
     SourceMaximum=-50
     SourceMinimum=-10
     TargetMinimum=0
     bClamp=true
     Value=&RigVMFunction_MathQuaternionToEuler(RotationOrder=ZYX, Value=@n4.Transform.Rotation).Result.Z
    ).Result
    BoneToModify.0.Transform.Rotation=&RigVMFunction_MathQuaternionFromEuler(Euler=&RigVMFunction_MathVectorAdd(B=(X=0,Y=30,Z=0), A=@n6.Result) {template="Add::Execute(in A,in B,out Result)"}.Result).Result
    BoneToModify.0.Transform.Translation=&RigVMFunction_MathVectorAdd(B=(X=-2,Y=0,Z=-5), A=@n5.Transform.Translation) {template="Add::Execute(in A,in B,out Result)"}.Result
    ExecutePin=@RigUnit_BeginExecution.ExecutePin
   )
   n7 = &RigUnit_GetRelativeBoneTransform(Bone=head, Space=spine_05)
   n8 = &RigUnit_GetInitialBoneTransform(Bone=dyn_towel_bk_r, Space=LocalSpace)
   n9 = &RigVMFunction_MathQuaternionToEuler(RotationOrder=ZYX, Value=@n8.Transform.Rotation)
   n10 = &RigUnit_GetInitialBoneTransform(Bone=dyn_towel_bk, Space=LocalSpace)
   n11 = &RigVMFunction_MathQuaternionToEuler(RotationOrder=ZYX, Value=@n10.Transform.Rotation)
   > &RigUnit_ModifyBoneTransforms(
    BoneToModify=((Bone="dyn_towel_bk_r",Transform=(Rotation=(X=0,Y=0,Z=0,W=1),Translation=(X=0,Y=0,Z=0),Scale3D=(X=1.3,Y=1,Z=0.3))),(Bone="dyn_towel_bk",Transform=(Rotation=(X=0,Y=0,Z=0,W=1),Translation=(X=0,Y=0,Z=0),Scale3D=(X=1,Y=1,Z=1))))
    Mode=OverrideLocal
    WeightMaximum=1
    Weight=&RigVMFunction_MathFloatRemap(
     SourceMaximum=-30
     SourceMinimum=-10
     TargetMinimum=0
     bClamp=true
     Value=&RigVMFunction_MathQuaternionToEuler(RotationOrder=YZX, Value=@n7.Transform.Rotation).Result.Y
    ).Result
    BoneToModify.0.Transform.Rotation=&RigVMFunction_MathQuaternionFromEuler(Euler=&RigVMFunction_MathVectorAdd(B=(X=-12,Y=0,Z=0), A=@n9.Result) {template="Add::Execute(in A,in B,out Result)"}.Result).Result
    BoneToModify.1.Transform.Rotation=&RigVMFunction_MathQuaternionFromEuler(Euler=&RigVMFunction_MathVectorAdd(B=(X=0,Y=0,Z=0), A=@n11.Result) {template="Add::Execute(in A,in B,out Result)"}.Result).Result
    BoneToModify.0.Transform.Translation=&RigVMFunction_MathVectorAdd(B=(X=1.6,Y=0.25,Z=-3), A=@n8.Transform.Translation) {template="Add::Execute(in A,in B,out Result)"}.Result
    BoneToModify.1.Transform.Translation=&RigVMFunction_MathVectorAdd(B=(X=0,Y=-0.3,Z=-3.8), A=@n10.Transform.Translation) {template="Add::Execute(in A,in B,out Result)"}.Result
   )
   n12 = &RigUnit_GetRelativeBoneTransform(Bone=upperarm_l, Space=spine_05)
   &RigVMFunction_MathFloatRemap(
    SourceMaximum=50
    SourceMinimum=170
    TargetMinimum=0
    bClamp=true
    Value=&RigVMFunction_MathFloatAbs(Value=&RigVMFunction_MathQuaternionToEuler(RotationOrder=ZYX, Value=@n12.Transform.Rotation).Result.Z) {template="Absolute::Execute(in Value,out Result)"}.Result
   )
   RigUnit_GetInitialBoneTransform = &RigUnit_GetInitialBoneTransform(Bone=dyn_towel_r_00, Space=LocalSpace)
   &RigVMFunction_MathQuaternionFromEuler(Euler=&RigVMFunction_MathVectorAdd(
     B=(X=0,Y=0,Z=10)
     A=&RigVMFunction_MathQuaternionToEuler(RotationOrder=ZYX, Value=@RigUnit_GetInitialBoneTransform.Transform.Rotation).Result
    ) {template="Add::Execute(in A,in B,out Result)"}.Result)
   n13 = &RigUnit_GetInitialBoneTransform(Bone=dyn_towel_r_01, Space=LocalSpace)
   &RigVMFunction_MathQuaternionFromEuler(Euler=&RigVMFunction_MathVectorAdd(B=(X=0,Y=0,Z=0), A=&RigVMFunction_MathQuaternionToEuler(RotationOrder=ZYX, Value=@n13.Transform.Rotation).Result) {template="Add::Execute(in A,in B,out Result)"}.Result)
   n14 = &RigUnit_GetInitialBoneTransform(Bone=dyn_towel_r_02, Space=LocalSpace)
   &RigVMFunction_MathQuaternionFromEuler(Euler=&RigVMFunction_MathVectorAdd(B=(X=0,Y=0,Z=0), A=&RigVMFunction_MathQuaternionToEuler(RotationOrder=ZYX, Value=@n14.Transform.Rotation).Result) {template="Add::Execute(in A,in B,out Result)"}.Result)
   &RigVMFunction_MathVectorAdd(B=(X=0,Y=0,Z=0), A=@RigUnit_GetInitialBoneTransform.Transform.Translation) {template="Add::Execute(in A,in B,out Result)"}
   &RigVMFunction_MathVectorAdd(B=(X=0,Y=0,Z=0), A=@n13.Transform.Translation) {template="Add::Execute(in A,in B,out Result)"}
   &RigVMFunction_MathVectorAdd(B=(X=0,Y=0,Z=0), A=@n14.Transform.Translation) {template="Add::Execute(in A,in B,out Result)"}
  ]}
]
hierarchy [
 root {loc=<0 0 0>, rot=<0 0 0 1>}
 pelvis {parent=root, loc=<0 -0.7071 93.1191>, rot=<0.0127 0.707 -0.0127 -0.707>}
 spine_01 {parent=pelvis, loc=<6.7494 -0.6388 0>, rot=<0 0 0.034 -0.9994>}
 spine_02 {parent=spine_01, loc=<7.2518 -0.035 0>, rot=<0 0 0.0238 -0.9997>}
 spine_03 {parent=spine_02, loc=<6.0104 0 0>, rot=<0 0 0.0149 -0.9999>}
 spine_04 {parent=spine_03, loc=<8.1024 0 0>, rot=<0 0 -0.0132 -0.9999>}
 spine_05 {parent=spine_04, loc=<10.0146 0 0>, rot=<0 0 -0.0015 -1>}
 clavicle_l {parent=spine_05, loc=<4.3575 -0.1466 -2.3223>, rot=<0.0465 -0.7289 -0.009 -0.6829>}
 upperarm_l {parent=clavicle_l, loc=<15.5111 0 0>, rot=<0.0226 -0.3234 -0.0088 -0.9459>}
 lowerarm_l {parent=upperarm_l, loc=<26.6514 0 0>, rot=<0.0391 -0.0099 -0.2453 -0.9686>}
 hand_l {parent=lowerarm_l, loc=<22.3491 0 0>, rot=<-0.7081 -0.0507 0.0975 -0.6975>}
 index_metacarpal_l {parent=hand_l, loc=<1.6535 0.0183 -2.96>, rot=<0.0158 -0.0124 -0.0173 -0.9996>}
 index_01_l {parent=index_metacarpal_l, loc=<8.3008 0.0357 -0.0524>, rot=<-0.0044 0.0485 0.0924 -0.9945>}
 index_02_l {parent=index_01_l, loc=<3.3587 -0.3004 0.0452>, rot=<-0.002 0.0027 0.3225 -0.9466>}
 index_03_l {parent=index_02_l, loc=<2.5578 -0.123 -0.0252>, rot=<-0.002 0.0067 0.2558 -0.9667>}
 middle_metacarpal_l {parent=hand_l, loc=<2.1856 -0.1725 -0.5984>, rot=<0.006 0.0168 -0.0329 -0.9993>}
 middle_01_l {parent=middle_metacarpal_l, loc=<7.3275 -0.2447 0.0129>, rot=<-0.0208 0.0042 0.0858 -0.9961>}
 middle_02_l {parent=middle_01_l, loc=<4.0123 0.1401 0.1175>, rot=<-0.0218 0.0258 0.3512 -0.9357>}
 middle_03_l {parent=middle_02_l, loc=<2.6077 0.0534 0.1373>, rot=<0.0082 0.0188 0.1708 -0.9851>}
 pinky_metacarpal_l {parent=hand_l, loc=<1.7616 -0.9753 3.3814>, rot=<-0.0397 0.0814 -0.069 -0.9935>}
 pinky_01_l {parent=pinky_metacarpal_l, loc=<7.1709 -0.4831 -0.1094>, rot=<-0.0654 -0.0035 0.2242 -0.9723>}
 pinky_02_l {parent=pinky_01_l, loc=<2.8274 0.0893 -0.1043>, rot=<0.0037 0.0011 0.2854 -0.9584>}
 pinky_03_l {parent=pinky_02_l, loc=<2.2046 0.0307 -0.0078>, rot=<-0.0002 0.0098 0.1445 -0.9895>}
 ring_metacarpal_l {parent=hand_l, loc=<2.2074 -0.3625 1.3638>, rot=<-0.0284 0.0722 -0.0458 -0.9959>}
 ring_01_l {parent=ring_metacarpal_l, loc=<7.0295 -0.6197 -0.1446>, rot=<-0.0489 -0.0252 0.1892 -0.9804>}
 ring_02_l {parent=ring_01_l, loc=<3.6174 0.4647 -0.1027>, rot=<0 0 0.2806 -0.9598>}
 ring_03_l {parent=ring_02_l, loc=<2.4152 0.114 0.0199>, rot=<0 0 0.1189 -0.9929>}
 thumb_01_l {parent=hand_l, loc=<1.3118 -0.3191 -2.459>, rot=<0.6546 -0.4042 0.0664 -0.6354>}
 thumb_02_l {parent=thumb_01_l, loc=<5.9336 -0.0448 0.0401>, rot=<-0.0037 0.0213 0.298 -0.9543>}
 thumb_03_l {parent=thumb_02_l, loc=<2.2223 0.3682 -0.0907>, rot=<0.0118 -0.013 0.0066 -0.9998>}
 lowerarm_twist_01_l {parent=lowerarm_l, loc=<17.8549 0 0>, rot=<0 0 0 -1>}
 lowerarm_twist_02_l {parent=lowerarm_l, loc=<9.6199 0 0>, rot=<0 0 0 -1>}
 deform_elbow_bk_l {parent=lowerarm_l, loc=<-1.2853 -2.9274 -0.5607>, rot=<-0.0285 0.0335 -0.0024 -0.999>}
 upperarm_twist_01_l {parent=upperarm_l, loc=<0.0041 0 0>, rot=<0 0 0 -1>}
 deform_deltoid_fr_l {parent=upperarm_twist_01_l, loc=<2.4511 0 0>, rot=<0 0 0 -1>}
 upperarm_twist_02_l {parent=upperarm_l, loc=<13.7941 0 0>, rot=<0 0 0 -1>}
 dyn_bicep_bk_base_l {parent=upperarm_twist_02_l, loc=<0 -11.1732 0>, rot=<0 0 0 -1>}
 dyn_bicep_bk_l {parent=dyn_bicep_bk_base_l, loc=<0 0 0>, rot=<0 0 0 -1>}
 deform_elbow_fr_l {parent=upperarm_l, loc=<26.5048 0.1875 -0.5816>, rot=<0.0021 0.0156 -0.185 -0.9826>}
 dyn_arm_measure_l {parent=clavicle_l, loc=<36.5819 0.0555 -16.3191>, rot=<0.0008 0.0336 -0.0252 -0.9991>}
 clavicle_r {parent=spine_05, loc=<4.3575 -0.1466 2.3223>, rot=<0.7289 0.0465 -0.6829 0.009>}
 upperarm_r {parent=clavicle_r, loc=<-15.5111 0 0>, rot=<0.0226 -0.3234 -0.0088 -0.9459>}
 lowerarm_r {parent=upperarm_r, loc=<-26.6515 0 0>, rot=<0.0391 -0.0099 -0.2453 -0.9686>}
 hand_r {parent=lowerarm_r, loc=<-22.3492 0 0>, rot=<-0.7081 -0.0507 0.0975 -0.6975>}
 index_metacarpal_r {parent=hand_r, loc=<-1.6534 -0.0181 2.9601>, rot=<0.0158 -0.0124 -0.0173 -0.9996>}
 index_01_r {parent=index_metacarpal_r, loc=<-8.3008 -0.0357 0.0524>, rot=<-0.0044 0.0485 0.0924 -0.9945>}
 index_02_r {parent=index_01_r, loc=<-3.3589 0.3003 -0.0453>, rot=<-0.002 0.0027 0.3225 -0.9466>}
 index_03_r {parent=index_02_r, loc=<-2.5578 0.1231 0.0251>, rot=<-0.002 0.0067 0.2558 -0.9667>}
 middle_metacarpal_r {parent=hand_r, loc=<-2.1856 0.1726 0.5984>, rot=<0.006 0.0168 -0.0329 -0.9993>}
 middle_01_r {parent=middle_metacarpal_r, loc=<-7.3273 0.2447 -0.0129>, rot=<-0.0208 0.0042 0.0858 -0.9961>}
 middle_02_r {parent=middle_01_r, loc=<-4.0142 -0.1411 -0.07>, rot=<-0.0217 0.0259 0.3512 -0.9357>}
 middle_03_r {parent=middle_02_r, loc=<-2.6074 -0.0534 -0.1312>, rot=<0.0083 0.0188 0.1708 -0.9851>}
 pinky_metacarpal_r {parent=hand_r, loc=<-1.7617 0.9753 -3.3814>, rot=<-0.0397 0.0814 -0.069 -0.9935>}
 pinky_01_r {parent=pinky_metacarpal_r, loc=<-7.1707 0.4832 0.1094>, rot=<-0.0654 -0.0035 0.2242 -0.9723>}
 pinky_02_r {parent=pinky_01_r, loc=<-2.8274 -0.0892 0.1043>, rot=<0.0037 0.0011 0.2854 -0.9584>}
 pinky_03_r {parent=pinky_02_r, loc=<-2.2045 -0.031 0.0078>, rot=<-0.0002 0.0098 0.1445 -0.9895>}
 ring_metacarpal_r {parent=hand_r, loc=<-2.2073 0.3627 -1.3638>, rot=<-0.0284 0.0722 -0.0458 -0.9959>}
 ring_01_r {parent=ring_metacarpal_r, loc=<-7.0297 0.6197 0.1446>, rot=<-0.0489 -0.0252 0.1892 -0.9804>}
 ring_02_r {parent=ring_01_r, loc=<-3.6171 -0.4648 0.1027>, rot=<0 0 0.2806 -0.9598>}
 ring_03_r {parent=ring_02_r, loc=<-2.4153 -0.1141 -0.0199>, rot=<0 0 0.1189 -0.9929>}
 thumb_01_r {parent=hand_r, loc=<-1.3119 0.3192 2.459>, rot=<0.6546 -0.4042 0.0664 -0.6354>}
 thumb_02_r {parent=thumb_01_r, loc=<-5.9336 0.0447 -0.0401>, rot=<-0.0037 0.0213 0.298 -0.9543>}
 thumb_03_r {parent=thumb_02_r, loc=<-2.2223 -0.3682 0.0907>, rot=<0.0118 -0.013 0.0066 -0.9998>}
 lowerarm_twist_01_r {parent=lowerarm_r, loc=<-17.855 0 0>, rot=<0 0 0 1>}
 lowerarm_twist_02_r {parent=lowerarm_r, loc=<-9.6199 0 0>, rot=<0 0 0 1>}
 deform_elbow_bk_r {parent=lowerarm_r, loc=<1.2852 2.9274 0.561>, rot=<-0.0285 0.0335 -0.0024 -0.999>}
 upperarm_twist_01_r {parent=upperarm_r, loc=<-0.0041 0 0>, rot=<0 0 0 1>}
 deform_deltoid_fr_r {parent=upperarm_twist_01_r, loc=<-2.4512 0 0>, rot=<0 0 0 1>}
 upperarm_twist_02_r {parent=upperarm_r, loc=<-13.7942 0 0>, rot=<0 0 0 1>}
 dyn_bicep_bk_base_r {parent=upperarm_twist_02_r, loc=<0 11.1732 0>, rot=<0 0 0 1>}
 dyn_bicep_bk_r {parent=dyn_bicep_bk_base_r, loc=<0 0 0>, rot=<0 0 0 1>}
 deform_elbow_fr_r {parent=upperarm_r, loc=<-26.5046 -0.1875 0.5813>, rot=<0.0021 0.0156 -0.185 -0.9826>}
 dyn_arm_measure_r {parent=clavicle_r, loc=<-36.5818 -0.0555 16.3194>, rot=<0.0008 0.0336 -0.0252 -0.9991>}
 deform_pectoral_fr_r {parent=clavicle_r, loc=<-9.39 0 0>, rot=<0 0 0 1>}
 neck_01 {parent=spine_05, loc=<11.3138 0.1529 0>, rot=<0 0 -0.2003 -0.9797>}
 neck_02 {parent=neck_01, loc=<4.5627 0 0>, rot=<0 0 0.0504 -0.9987>}
 head {parent=neck_02, loc=<5.3914 0 0>, rot=<0 0 0.115 -0.9934>}
 dyn_necklace_MSTR {parent=spine_05, loc=<9.4815 7.617 0>, rot=<0.242 0.6644 0.6644 0.242>}
 dyn_necklace_01 {parent=dyn_necklace_MSTR, loc=<0 0 0>, rot=<0 -0.007 0 -1>}
 dyn_necklace_02 {parent=dyn_necklace_01, loc=<9.1465 0 0>, rot=<0 -0.1194 0.0003 -0.9928>}
 dyn_whistle {parent=dyn_necklace_02, loc=<8.3497 0.0054 -0.1443>, rot=<0.0001 -0.2598 -0.0003 -0.9657>}
 dyn_towel_l_MSTR {parent=spine_05, loc=<17.0082 1.8102 -10.259>, rot=<-0.2359 -0.7212 -0.5652 -0.3236>}
 dyn_towel_l_00 {parent=dyn_towel_l_MSTR, loc=<0 0 0>, rot=<0 0 0 1>}
 dyn_towel_l_01 {parent=dyn_towel_l_00, loc=<7.2008 0 0>, rot=<0.0886 -0.0933 -0.0644 -0.9896>}
 dyn_towel_l_02 {parent=dyn_towel_l_01, loc=<10.1232 0 0>, rot=<0.0007 -0.0169 -0.0159 -0.9997>}
 dyn_towel_l_03 {parent=dyn_towel_l_02, loc=<8.3817 0 0>, rot=<-0.0018 -0.0964 -0.019 -0.9952>}
 dyn_towel_r_MSTR {parent=spine_05, loc=<17.171 2.3403 10.1895>, rot=<0.3505 0.5285 0.7482 0.195>}
 dyn_towel_r_00 {parent=dyn_towel_r_MSTR, loc=<0 0 0>, rot=<0 0 0 1>}
 dyn_towel_r_01 {parent=dyn_towel_r_00, loc=<6.4984 0 0>, rot=<-0.1166 -0.0603 0.0695 -0.9889>}
 dyn_towel_r_02 {parent=dyn_towel_r_01, loc=<10.2103 -0.1368 -0.4792>, rot=<-0.0143 -0.0831 0.025 -0.9961>}
 dyn_towel_r_03 {parent=dyn_towel_r_02, loc=<8.4362 -0.0657 -0.3389>, rot=<-0.002 -0.1008 0.0197 -0.9947>}
 dyn_towel_bk_r_MSTR {parent=spine_05, loc=<19.0155 -1.1226 5.454>, rot=<0.7254 -0.0893 0.5283 -0.4321>}
 dyn_towel_bk_r {parent=dyn_towel_bk_r_MSTR, loc=<0 0 0>, rot=<0 0 0 1>}
 dyn_towel_bk_MSTR {parent=spine_05, loc=<19.981 -3.0548 0.3849>, rot=<0.6329 -0.2935 0.3529 -0.6235>}
 dyn_towel_bk {parent=dyn_towel_bk_MSTR, loc=<0 0 0>, rot=<0 0 0 1>}
 dyn_towel_bk_l_MSTR {parent=spine_05, loc=<19.0119 -1.0188 -6.2222>, rot=<0.4464 -0.5273 0.0775 -0.7188>}
 dyn_towel_bk_l {parent=dyn_towel_bk_l_MSTR, loc=<0 0 0>, rot=<0 0 0 1>}
 thigh_l {parent=pelvis, loc=<-2.5693 -0.0109 -7.7501>, rot=<0.0012 0.0572 0.0155 -0.9982>}
 calf_l {parent=thigh_l, loc=<-39.0484 0 0>, rot=<0.0006 -0.0161 -0.0391 -0.9991>}
 calf_twist_01_l {parent=calf_l, loc=<-35.813 -0.0967 0.0397>, rot=<0.0237 0.0095 -0.0035 -0.9997>}
 calf_twist_02_l {parent=calf_l, loc=<-18.8064 -0.0967 0.0397>, rot=<0.0237 0.0095 -0.0035 -0.9997>}
 foot_l {parent=calf_l, loc=<-42.561 0.1077 0.4292>, rot=<0.0418 -0.0455 0.0343 -0.9975>}
 ball_l {parent=foot_l, loc=<-6.1599 13.8264 0.6806>, rot=<0.0001 -0.0001 -0.7063 -0.7079>}
 deform_knee_bk_01_l {parent=calf_l, loc=<0.6022 -1.0607 0>, rot=<0 0 -0.0408 -0.9992>}
 deform_knee_fr_01_l {parent=calf_l, loc=<-0.2342 0 0>, rot=<0 0 0 1>}
 thigh_twist_01_l {parent=thigh_l, loc=<-21.6303 0 0>, rot=<0.0013 0 0 -1>}
 deform_thigh_bk_l {parent=thigh_twist_01_l, loc=<16.9514 0.2058 0.1808>, rot=<-0.0013 -0.0217 0 -0.9998>}
 deform_thigh_fr_l {parent=thigh_twist_01_l, loc=<21.5084 0.2281 -0.0598>, rot=<-0.0013 -0.0217 0 -0.9998>}
 thigh_r {parent=pelvis, loc=<-2.5691 -0.0109 7.7501>, rot=<0.0572 -0.0012 0.9982 0.0155>}
 calf_r {parent=thigh_r, loc=<39.0487 0 0>, rot=<0.0006 -0.0161 -0.0391 -0.9991>}
 calf_twist_01_r {parent=calf_r, loc=<35.8132 0.0967 -0.0397>, rot=<0.0237 0.0095 -0.0035 -0.9997>}
 calf_twist_02_r {parent=calf_r, loc=<18.8064 0.0967 -0.0397>, rot=<0.0237 0.0095 -0.0035 -0.9997>}
 foot_r {parent=calf_r, loc=<42.5612 -0.1077 -0.4292>, rot=<0.0418 -0.0455 0.0343 -0.9975>}
 ball_r {parent=foot_r, loc=<6.1598 -13.8264 -0.6806>, rot=<0.0001 -0.0001 -0.7063 -0.7079>}
 dyn_volume_base_calf_r {parent=calf_r, loc=<0 -4 0>, rot=<0 0 0 1>}
 dyn_volume_calf_r {parent=dyn_volume_base_calf_r, loc=<0 0 0>, rot=<0 0 0 1>}
 deform_knee_bk_01_r {parent=calf_r, loc=<-0.6022 1.0607 0>, rot=<0 0 -0.0408 -0.9992>}
 deform_knee_fr_01_r {parent=calf_r, loc=<0.2341 0 0>, rot=<0 0 0 1>}
 thigh_twist_01_r {parent=thigh_r, loc=<21.6303 0 0>, rot=<0.0013 0 0 -1>}
 deform_thigh_bk_r {parent=thigh_twist_01_r, loc=<-16.9512 -0.2058 -0.1808>, rot=<-0.0013 -0.0217 0 -0.9998>}
 deform_thigh_fr_r {parent=thigh_twist_01_r, loc=<-21.5083 -0.2281 0.0598>, rot=<-0.0013 -0.0217 0 -0.9998>}
 def_belt_l_MSTR {parent=pelvis, loc=<4.5099 7.3908 -10.7027>, rot=<0.8946 0.0319 0.0566 -0.442>}
 def_belt_l {parent=def_belt_l_MSTR, loc=<0 0 0>, rot=<0 0 0 -1>}
 def_belt_r_MSTR {parent=pelvis, loc=<4.5099 7.3908 10.7027>, rot=<0.442 -0.0566 -0.0319 -0.8946>}
 def_belt_r {parent=def_belt_r_MSTR, loc=<0 0 0>, rot=<0 0 0 -1>}
 dyn_comb_MSTR {parent=pelvis, loc=<1.1888 -10.6645 9.671>, rot=<-0.5568 0.0045 -0.1035 -0.8242>}
 dyn_comb {parent=dyn_comb_MSTR, loc=<0 0 0>, rot=<0 0 0 -1>}
 dyn_rope_MSTR {parent=pelvis, loc=<3.8298 11.4536 0.1992>, rot=<-0.5089 -0.491 -0.491 -0.5089>}
 dyn_rope_01_r {parent=dyn_rope_MSTR, loc=<0.0273 0.2372 -1.1631>, rot=<-0.0554 -0.6614 -0.0625 -0.7453>}
 dyn_rope_02_r {parent=dyn_rope_01_r, loc=<4.8079 0 0>, rot=<0 0 0 -1>}
 dyn_rope_01_l {parent=dyn_rope_MSTR, loc=<-0.072 -0.1136 -1.4002>, rot=<0.0315 -0.659 0.0358 -0.7506>}
 dyn_rope_02_l {parent=dyn_rope_01_l, loc=<3.647 0 0>, rot=<0 0 0 1>}
 dyn_node_l {parent=dyn_rope_MSTR, loc=<0.3603 -0.8206 -1.0789>, rot=<0.0921 -0.6154 0.1159 -0.7742>}
 dyn_node_r {parent=dyn_rope_MSTR, loc=<0.4335 1.1221 -0.6>, rot=<-0.0778 -0.6624 -0.0869 -0.74>}
]
