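# Physics asset for the dangling rope on the Quarrel Summer medium male player body cosmetic.
# Small capsules form left and right two-segment rope chains and two side node bodies, all branching from a master rope bone linked to the pelvis.
# Rope and node joints limit swing and nearly lock twist, while master-to-pelvis and thigh-to-pelvis joints swing and twist freely.
# Large pelvis and thigh capsules approximate the hips and legs around the rope.
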PhysicsAsset /BRCosmetics/Characters/Player/Male/Medium/Bodies/M_MED_Quarrel_Summer/Meshes/M_MED_Quarrel_Summer_Summer_dyn_rope
bodies {
 dyn_node_l [{shapeType=Capsule, length=2, radius=2, center=<3 -0.45 1>, rotation=<rot 90 90 110.0042>}]
 dyn_node_r [{shapeType=Capsule, length=2, radius=1.5, center=<3.5 0.8 1>, rotation=<rot 0 20 90>}]
 dyn_rope_01_l [{shapeType=Capsule, length=1.5, radius=1, center=<2 0 1>, rotation=<rot 90 0 0>}]
 dyn_rope_01_r [{shapeType=Capsule, length=1.75, radius=1, center=<2 0 1>, rotation=<rot 90 0 0>}]
 dyn_rope_02_l [{shapeType=Capsule, length=2.25, radius=1, center=<3 0 1>, rotation=<rot 90 0 0>}]
 dyn_rope_02_r [{shapeType=Capsule, length=2.5, radius=1, center=<3.5 0 0.9999>, rotation=<rot 90 0 0>}]
 dyn_rope_MSTR [{shapeType=Capsule, length=2.2767, radius=0.7796, center=<0.0955 0.0254 -0.1812>, rotation=<rot 0 -178.9218 -87.9463>}]
 pelvis [{shapeType=Capsule, length=10.4405, radius=13.2413, center=<-5.1276 -1.1547 0.0002>}]
 thigh_l [{shapeType=Capsule, length=9.4405, radius=9.7413, center=<-6.8701 1.8336 -1.0258>, rotation=<rot 85 0 0>}]
 thigh_r [{shapeType=Capsule, length=9.4405, radius=9.7413, center=<6.8701 -1.8336 1.0258>, rotation=<rot 85 0 180>}]
}
constraints [
 {
  bone1Name=dyn_rope_MSTR
  bone2Name=pelvis
  swing1LimitDegrees=45
  swing1Motion=Free
  swing2LimitDegrees=45
  swing2Motion=Free
  twistLimitDegrees=45
  twistMotion=Free
 }
 {bone1Name=dyn_rope_01_r, bone2Name=dyn_rope_MSTR, swing1LimitDegrees=60, swing2LimitDegrees=45, twistLimitDegrees=2}
 {bone1Name=dyn_rope_02_r, bone2Name=dyn_rope_01_r, swing1LimitDegrees=60, swing2LimitDegrees=45, twistLimitDegrees=2}
 {bone1Name=dyn_rope_01_l, bone2Name=dyn_rope_MSTR, swing1LimitDegrees=60, swing2LimitDegrees=45, twistLimitDegrees=2}
 {bone1Name=dyn_rope_02_l, bone2Name=dyn_rope_01_l, swing1LimitDegrees=60, swing2LimitDegrees=45, twistLimitDegrees=2}
 {
  bone1Name=thigh_l
  bone2Name=pelvis
  swing1LimitDegrees=45
  swing1Motion=Free
  swing2LimitDegrees=45
  swing2Motion=Free
  twistLimitDegrees=45
  twistMotion=Free
 }
 {bone1Name=dyn_node_l, bone2Name=dyn_rope_MSTR, swing1LimitDegrees=60, swing2LimitDegrees=35, twistLimitDegrees=2}
 {bone1Name=dyn_node_r, bone2Name=dyn_rope_MSTR, swing1LimitDegrees=60, swing2LimitDegrees=35, twistLimitDegrees=2}
 {
  bone1Name=thigh_r
  bone2Name=pelvis
  swing1LimitDegrees=45
  swing1Motion=Free
  swing2LimitDegrees=45
  swing2Motion=Free
  twistLimitDegrees=45
  twistMotion=Free
 }
]
