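# Physics asset for the dynamic necklace on the male medium Quarrel Summer character body, setting up its physics-driven motion against the torso.
# Two large spine capsules approximate the upper torso, while a tiny master capsule and two slimmer capsules form the necklace chain.
# The master-to-spine_05 and spine_05-to-spine_04 joints rotate freely.
# The two necklace links have narrow swing and very tight twist limits, so the chain bends slightly with little roll.
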PhysicsAsset /BRCosmetics/Characters/Player/Male/Medium/Bodies/M_MED_Quarrel_Summer/Meshes/M_MED_Quarrel_Summer_dyn_necklace
bodies {
 dyn_necklace_01 [{shapeType=Capsule, length=8.75, radius=2.5, center=<5.0066 0 2.5759>, rotation=<rot 90 -0.0198 -0.0198>}]
 dyn_necklace_02 [{shapeType=Capsule, length=9.75, radius=2.5, center=<4.836 -0.0004 2.0316>, rotation=<rot 90 0.0099 0.0099>}]
 dyn_necklace_MSTR [{shapeType=Capsule, length=0.5, radius=0.5}]
 spine_04 [{shapeType=Capsule, length=14.2009, radius=13.4611, center=<2.4005 3.1688 0>, rotation=<rot 0 90 0>}]
 spine_05 [{shapeType=Capsule, length=18.5759, radius=12.4611, center=<-1.4577 1.0496 0>, rotation=<rot 0 90 0>}]
}
constraints [
 {
  bone1Name=dyn_necklace_MSTR
  bone2Name=spine_05
  swing1LimitDegrees=45
  swing1Motion=Free
  swing2LimitDegrees=45
  swing2Motion=Free
  twistLimitDegrees=45
  twistMotion=Free
 }
 {bone1Name=dyn_necklace_01, bone2Name=dyn_necklace_MSTR, swing1LimitDegrees=15, swing2LimitDegrees=15, twistLimitDegrees=2}
 {bone1Name=dyn_necklace_02, bone2Name=dyn_necklace_01, swing1LimitDegrees=15, swing2LimitDegrees=15, twistLimitDegrees=2}
 {
  bone1Name=spine_05
  bone2Name=spine_04
  swing1LimitDegrees=45
  swing1Motion=Free
  swing2LimitDegrees=45
  swing2Motion=Free
  twistLimitDegrees=45
  twistMotion=Free
 }
]
