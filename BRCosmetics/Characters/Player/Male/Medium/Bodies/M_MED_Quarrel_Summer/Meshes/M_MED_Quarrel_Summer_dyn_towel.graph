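# Physics asset for the towel on the Quarrel Summer male medium body cosmetic, giving capsule bodies and constraints to its left and right dynamic bones.
# Each side chains four towel bones from a small master body linked to spine_05, with narrow swing and very tight twist limits that differ slightly per side.
# Spine, upper-arm twist and necklace capsules form the surrounding body shapes, joined through spine_05 by free-rotation constraints.
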PhysicsAsset /BRCosmetics/Characters/Player/Male/Medium/Bodies/M_MED_Quarrel_Summer/Meshes/M_MED_Quarrel_Summer_dyn_towel
bodies {
 dyn_necklace_02 [{shapeType=Capsule, length=5.762, radius=7.4828, shapeName=spine_04_capsule, center=<6.2354 0 -5.4173>, rotation=<rot 84.9383 -180 -180>}]
 dyn_towel_l_00 [{shapeType=Capsule, length=4.25, radius=2.75, shapeName=dyn_towel_r_01_capsule, center=<4 0 0.6>, rotation=<rot 90 0 0>}]
 dyn_towel_l_01 [{shapeType=Capsule, length=5.25, radius=3.5, shapeName=dyn_towel_r_01_capsule, center=<4.9569 0 2>, rotation=<rot 84.9892 180 180>}]
 dyn_towel_l_02 [{shapeType=Capsule, length=4.75, radius=4, shapeName=dyn_towel_r_02_capsule, center=<5.2006 0 2.3>, rotation=<rot 84.9994 0 0>}]
 dyn_towel_l_03 [{shapeType=Capsule, length=4.25, radius=4.5, shapeName=dyn_towel_r_03_capsule, center=<5.9999 0 0>, rotation=<rot 0 0 74.997>}]
 dyn_towel_l_MSTR [{shapeType=Capsule, length=0.505, radius=0.505, shapeName=dyn_towel_l_MSTR_capsule, rotation=<rot 0 -90 0>}]
 dyn_towel_r_00 [{shapeType=Capsule, length=4.25, radius=3.25, shapeName=dyn_towel_r_01_capsule, center=<4 0 1>, rotation=<rot 84.9957 180 -180>}]
 dyn_towel_r_01 [{shapeType=Capsule, length=4, radius=3.75, shapeName=dyn_towel_r_01_capsule, center=<7 0 1.2>, rotation=<rot 84.9998 0 0>}]
 dyn_towel_r_02 [{shapeType=Capsule, length=3.5, radius=4.25, shapeName=dyn_towel_r_02_capsule, center=<4.8807 0 2.3>, rotation=<rot 85 0 0>}]
 dyn_towel_r_03 [{shapeType=Capsule, length=4, radius=4.25, shapeName=dyn_towel_r_03_capsule, center=<7 -1.5 0.8>, rotation=<rot 0 0 105.0002>}]
 dyn_towel_r_MSTR [{shapeType=Capsule, length=0.505, radius=0.505, shapeName=dyn_towel_r_MSTR_capsule, center=<0 0.8829 0>}]
 spine_04 [
  {shapeType=Capsule, length=16.512, radius=15.7328, shapeName=spine_04_capsule, center=<1.8802 0.5354 -0.0001>}
  {shapeType=Capsule, length=36.762, radius=13.9828, shapeName=spine_04_capsule, center=<5.4028 -0.9979 0>}
 ]
 spine_05 [{shapeType=Capsule, length=25.0973, radius=13.973, shapeName=spine_05_capsule, center=<1.1492 -4.2359 0>}]
 upperarm_twist_01_l [{shapeType=Capsule, length=15.0973, radius=6.973, shapeName=spine_05_capsule, center=<13.4382 -7.7322 2.4152>, rotation=<rot 90 0 174.9982>}]
 upperarm_twist_01_r [{shapeType=Capsule, length=18.0973, radius=5.973, shapeName=spine_05_capsule, center=<-10.4871 7.6244 -3.1036>, rotation=<rot 90 0 -5.0025>}]
}
constraints [
 {
  bone1Name=spine_05
  bone2Name=spine_04
  swing1LimitDegrees=45
  swing1Motion=Free
  swing2LimitDegrees=45
  swing2Motion=Free
  twistLimitDegrees=45
  twistMotion=Free
 }
 {
  bone1Name=dyn_towel_l_MSTR
  bone2Name=spine_05
  swing1LimitDegrees=45
  swing1Motion=Free
  swing2LimitDegrees=45
  swing2Motion=Free
  twistLimitDegrees=45
  twistMotion=Free
 }
 {bone1Name=dyn_towel_l_01, bone2Name=dyn_towel_l_00, swing1LimitDegrees=15, swing2LimitDegrees=15, twistLimitDegrees=2}
 {bone1Name=dyn_towel_l_02, bone2Name=dyn_towel_l_01, swing1LimitDegrees=15, swing2LimitDegrees=15, twistLimitDegrees=2}
 {bone1Name=dyn_towel_l_03, bone2Name=dyn_towel_l_02, swing1LimitDegrees=15, swing2LimitDegrees=15, twistLimitDegrees=2}
 {
  bone1Name=dyn_towel_r_MSTR
  bone2Name=spine_05
  swing1LimitDegrees=45
  swing1Motion=Free
  swing2LimitDegrees=45
  swing2Motion=Free
  twistLimitDegrees=45
  twistMotion=Free
 }
 {bone1Name=dyn_towel_r_01, bone2Name=dyn_towel_r_00, swing1LimitDegrees=20, swing2LimitDegrees=15, twistLimitDegrees=2}
 {bone1Name=dyn_towel_r_02, bone2Name=dyn_towel_r_01, swing1LimitDegrees=20, swing2LimitDegrees=15, twistLimitDegrees=2}
 {bone1Name=dyn_towel_r_03, bone2Name=dyn_towel_r_02, swing1LimitDegrees=20, swing2LimitDegrees=15, twistLimitDegrees=2}
 {bone1Name=dyn_towel_r_00, bone2Name=dyn_towel_r_MSTR, swing1LimitDegrees=15, swing2LimitDegrees=15, twistLimitDegrees=2}
 {bone1Name=dyn_towel_l_00, bone2Name=dyn_towel_l_MSTR, swing1LimitDegrees=15, swing2LimitDegrees=15, twistLimitDegrees=2}
 {
  bone1Name=dyn_necklace_02
  bone2Name=spine_05
  swing1LimitDegrees=45
  swing1Motion=Free
  swing2LimitDegrees=45
  swing2Motion=Free
  twistLimitDegrees=45
  twistMotion=Free
 }
 {
  bone1Name=upperarm_twist_01_r
  bone2Name=spine_05
  swing1LimitDegrees=45
  swing1Motion=Free
  swing2LimitDegrees=45
  swing2Motion=Free
  twistLimitDegrees=45
  twistMotion=Free
 }
 {
  bone1Name=upperarm_twist_01_l
  bone2Name=spine_05
  swing1LimitDegrees=45
  swing1Motion=Free
  swing2LimitDegrees=45
  swing2Motion=Free
  twistLimitDegrees=45
  twistMotion=Free
 }
]
