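# Cosmetic animation Blueprint for the Quarrel Summer body, adding deformation and secondary motion to the pose copied from the parent mesh.
# Deform, sculpt and towel Control Rigs lead into constraints, RigidBody sims for necklace, rope and towel, and whistle AnimDynamics.
# Additive bone corrections around the sims fix biceps, rope and towel, using alphas measured from arm and thigh poses.
# Rifle-type weapons enable an extra arm-height-driven right-towel correction.
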AnimBlueprint /BRCosmetics/Characters/Player/Male/Medium/Bodies/M_MED_Quarrel_Summer/Meshes/M_MED_Quarrel_Summer_AnimBP
target_skeleton=&/M_MED_Quarrel_Summer_Skeleton
graphs [
 AnimGraph {nodes [
   n1 = "Constraint - dyn_towel_r_MSTR (PP)"!(ComponentPose, Alpha=1, ConstraintWeights_0=0.8, ConstraintWeights_1:float=0.2) -> (Pose) {
    alphaBoolBlend {blendOption=Linear}
    boneToModify {boneName=dyn_towel_r_MSTR}
    constraintSetup [{targetBone {boneName=spine_05}}, {targetBone {boneName=clavicle_r}}]
    constraintWeights [1,1]
   }
   "Constraint - dyn_towel_l_MSTR (PP)"!(ComponentPose=@n1, Alpha=1, ConstraintWeights_0=0.8, ConstraintWeights_1:float=0.2) -> (Pose) {
    alphaBoolBlend {blendOption=Linear}
    boneToModify {boneName=dyn_towel_l_MSTR}
    constraintSetup [{targetBone {boneName=spine_05}}, {targetBone {boneName=clavicle_l}}]
    constraintWeights [1,1]
   }
   n2 = "Transform (Modify) Bone - Bone: def_belt_l"(Alpha=Get thigh_rot_l(), Translation="10,0,1", Rotation="0,-15,0", Scale="1,1,1") {
    rotationMode=BMM_Additive
    rotationSpace=BCS_BoneSpace
    translationMode=BMM_Additive
    translationSpace=BCS_BoneSpace
    alphaBoolBlend {blendOption=Linear}
    boneToModify {boneName=def_belt_l}
    rotation {yaw=-15}
    translation {x=10, z=1}
   }
   "Transform (Modify) Bone - Bone: def_belt_r"(ComponentPose=@n2, Alpha=Get thigh_rot_r(), Translation="10,0,0", Rotation="0,10,0", Scale="1,1,1") {
    rotationMode=BMM_Additive
    rotationSpace=BCS_BoneSpace
    translationMode=BMM_Additive
    translationSpace=BCS_BoneSpace
    alphaBoolBlend {blendOption=Linear}
    boneToModify {boneName=def_belt_r}
    rotation {yaw=10}
    translation {x=10}
   }
   n3 = Control Rig(Source=Copy Pose From Mesh() {bUseAttachedParent=true}) {
    alphaBoolBlend {blendOption=Linear}
    controlRigAssetReference {blueprintRigClass=/BRCosmetics/Characters/Player/Male/Medium/Bodies/M_MED_Quarrel/Meshes/M_MED_Quarrel_Deform_CtrlRig_twists.M_MED_Quarrel_Deform_CtrlRig_twists_C}
   }
   n4 = Control Rig!(
    Source=@n3
    Alpha
    KneeRot_fr_l_114deg:Vector
    KneeTrans_fr_l_114deg:Vector="0.258,-3.591,0.009"
    KneeScale_fr_l_114deg:Vector="1,1,1"
    KneeScale_fr_l_130deg:Vector="1,1,1"
    KneeRot_fr_l_130deg:Vector="0,0,-51.2"
    KneeTrans_fr_l_130deg:Vector="0.258,-3.591,0.009"
   ) -> (Pose) {
    alphaBoolBlend {blendOption=Linear}
    controlRigAssetReference {blueprintRigClass=/BRCosmetics/Characters/Player/Male/Medium/Bodies/M_MED_Quarrel/Meshes/M_MED_Quarrel_Deform_CtrlRig_sculpters.M_MED_Quarrel_Deform_CtrlRig_sculpters_C}
   }
   n5 = Control Rig!(Source=@n4, Alpha) -> (Pose) {
    alphaBoolBlend {blendOption=Linear}
    controlRigAssetReference {blueprintRigClass=&/M_MED_Quarrel_Summer_CtrlRig_towel.M_MED_Quarrel_Summer_CtrlRig_towel_C}
   }
   n6 = "Constraint - dyn_comb_MSTR (PPP)"!(
    ComponentPose=Local To Component(LocalPose=@n5)
    Alpha=1
    ConstraintWeights_0=0.8812
    ConstraintWeights_1:float=0.0084
    ConstraintWeights_2:float=0.1104
   ) -> (Pose) {
    alphaBoolBlend {blendOption=Linear}
    boneToModify {boneName=dyn_comb_MSTR}
    constraintSetup [{targetBone {boneName=pelvis}}, {targetBone {boneName=deform_thigh_bk_r}}, {targetBone {boneName=thigh_twist_01_r}}]
    constraintWeights [1,1,1]
   }
   n7 = "Transform (Modify) Bone - Bone: dyn_bicep_bk_l"(ComponentPose=@n6, Alpha=Get l_arm_fwd() -> (l_arm_fwd:double), Translation="0,10,0", Rotation="0,0,0", Scale="1,1,1") {
    rotationMode=BMM_Additive
    rotationSpace=BCS_BoneSpace
    scaleMode=BMM_Additive
    scaleSpace=BCS_BoneSpace
    translationMode=BMM_Additive
    translationSpace=BCS_BoneSpace
    alphaBoolBlend {blendOption=Linear}
    alphaScaleBiasClamp {inRange {max=0, min=-100}, outRange {max=0, min=1}}
    boneToModify {boneName=dyn_bicep_bk_l}
    translation {y=10}
   }
   n8 = "Transform (Modify) Bone - Bone: dyn_bicep_bk_r"(ComponentPose=@n7, Alpha=Get r_arm_fwd() -> (r_arm_fwd:double), Translation="0,-10,0", Rotation="0,0,0", Scale="1,1,1") {
    rotationMode=BMM_Additive
    rotationSpace=BCS_BoneSpace
    scaleMode=BMM_Additive
    scaleSpace=BCS_BoneSpace
    translationMode=BMM_Additive
    translationSpace=BCS_BoneSpace
    alphaBoolBlend {blendOption=Linear}
    alphaScaleBiasClamp {inRange {max=0, min=-100}, outRange {max=0, min=1}}
    boneToModify {boneName=dyn_bicep_bk_r}
    translation {y=-10}
   }
   n9 = RigidBody!(ComponentPose=@n8, Alpha, OverrideWorldGravity:Vector="0,0,-980", ExternalForce) -> (Pose) {
    bClampLinearTranslationLimitToRefPose=true
    bOverrideWorldGravity=true
    overridePhysicsAsset=&/M_MED_Quarrel_Summer_dyn_necklace
    simulationSpace=BaseBoneSpace
    alphaBoolBlend {blendOption=Linear}
    baseBoneRef {boneName=spine_05}
    componentAppliedLinearAccClamp=<2000 2000 1000>
    componentLinearAccScale {x=0.1}
    componentLinearVelScale {x=0.1}
    overrideWorldGravity {z=-980}
    simSpaceSettings {maxAngularAcceleration=50, maxAngularVelocity=2, maxLinearAcceleration=50, maxLinearVelocity=5, velocityScaleZ=0, worldAlpha=1}
   }
   n10 = Anim Dynamics - Bone: dyn_whistle(ComponentPose=@n9, Alpha=1, ExternalForce="0,0,0") {
    angularDampingOverride=0.9999
    bGravityOverrideInSimSpace=true
    bOverrideAngularDamping=true
    bUseGravityOverride=true
    simulationSpace=RootRelative
    alphaBoolBlend {blendOption=Linear}
    boundBone {boneName=dyn_whistle}
    componentAppliedLinearAccClamp=<3000 3000 3000>
    componentLinearAccScale=<0.3 0.3 0.3>
    componentLinearVelScale=<0.3 0.3 0.3>
    gravityOverride {z=-980}
    physicsBodyDefinitions [{
      sphereCollisionRadius=0
      boundBone {boneName=dyn_whistle}
      boxExtents=<10 8 8>
      constraintSetup {angularLimitsMax=<0 0 0>, angularLimitsMin=<0 -35 0>, angularTarget=<0 0 0>}
      localJointOffset=<1.5 0 0>
     }]
    relativeSpaceBone {boneName=dyn_necklace_02}
   }
   n11 = "Constraint - dyn_rope_MSTR (PPP)"!(ComponentPose=@n10, Alpha=1, ConstraintWeights_0=0.6817, ConstraintWeights_1:float=0.131, ConstraintWeights_2:float=0.1873) -> (Pose) {
    alphaBoolBlend {blendOption=Linear}
    boneToModify {boneName=dyn_rope_MSTR}
    constraintSetup [{targetBone {boneName=pelvis}}, {targetBone {boneName=def_belt_r}}, {targetBone {boneName=def_belt_l}}]
    constraintWeights [1,1,1]
   }
   n12 = "Transform (Modify) Bone - Bone: dyn_rope_MSTR"(ComponentPose=@n11, Alpha=Get thigh_rot_l(), Translation="0.000,1,0.000", Rotation="0,0,0", Scale="1,1,1") {
    rotationMode=BMM_Additive
    rotationSpace=BCS_BoneSpace
    scaleMode=BMM_Additive
    scaleSpace=BCS_BoneSpace
    translationMode=BMM_Additive
    translationSpace=BCS_BoneSpace
    alphaBoolBlend {blendOption=Linear}
    alphaScaleBiasClamp {inRange {max=0, min=-100}, outRange {max=0, min=1}}
    boneToModify {boneName=dyn_rope_MSTR}
    translation {y=1}
   }
   n13 = RigidBody(ComponentPose=@n12, OverrideWorldGravity="0,0,-980") {
    bClampLinearTranslationLimitToRefPose=true
    bOverrideWorldGravity=true
    overridePhysicsAsset=&/M_MED_Quarrel_Summer_Summer_dyn_rope
    simulationSpace=BaseBoneSpace
    alphaBoolBlend {blendOption=Linear}
    baseBoneRef {boneName=pelvis}
    componentAppliedLinearAccClamp=<1000 1000 1000>
    componentLinearAccScale {x=0.2}
    componentLinearVelScale {x=0.2}
    overrideWorldGravity {z=-980}
    simSpaceSettings {maxAngularAcceleration=100, maxAngularVelocity=2, maxLinearAcceleration=300, maxLinearVelocity=30, velocityScaleZ=0, worldAlpha=1}
   }
   n14 = "Transform (Modify) Bone - Bone: dyn_towel_r_MSTR"(ComponentPose=@n13, Alpha=Get r_arm_fwd(), Translation="0.221,-5.022,1.014", Rotation="13.619,-0.567,-6.488", Scale="1,1,1") {
    rotationMode=BMM_Additive
    rotationSpace=BCS_BoneSpace
    scaleMode=BMM_Additive
    scaleSpace=BCS_BoneSpace
    translationMode=BMM_Additive
    translationSpace=BCS_BoneSpace
    alphaBoolBlend {blendOption=Linear}
    alphaScaleBiasClamp {inRange {max=0, min=-100}, outRange {max=0, min=1}}
    boneToModify {boneName=dyn_towel_r_MSTR}
    rotation=<rot 13.619 -0.567 -6.4881>
    translation=<0.221 -5.022 1.014>
   }
   n15 = "Transform (Modify) Bone - Bone: dyn_towel_l_MSTR"(ComponentPose=@n14, Alpha=Get l_arm_fwd(), Translation="-0.079,2.541,0.908", Rotation="5.000,10.001,0.000", Scale="1,1,1") {
    rotationMode=BMM_Additive
    rotationSpace=BCS_BoneSpace
    scaleMode=BMM_Additive
    scaleSpace=BCS_BoneSpace
    translationMode=BMM_Additive
    translationSpace=BCS_BoneSpace
    alphaBoolBlend {blendOption=Linear}
    alphaScaleBiasClamp {inRange {max=0, min=-100}, outRange {max=0, min=1}}
    boneToModify {boneName=dyn_towel_l_MSTR}
    rotation {pitch=5, yaw=10.001}
    translation=<-0.079 2.541 0.908>
   }
   n16 = RigidBody(ComponentPose=@n15, OverrideWorldGravity="0,0,-980") {
    bClampLinearTranslationLimitToRefPose=true
    bOverrideWorldGravity=true
    overridePhysicsAsset=&/M_MED_Quarrel_Summer_dyn_towel
    simulationSpace=BaseBoneSpace
    alphaBoolBlend {blendOption=Linear}
    baseBoneRef {boneName=spine_05}
    componentAppliedLinearAccClamp=<2000 2000 1000>
    componentLinearAccScale {x=0.2}
    componentLinearVelScale {x=0.2}
    overrideWorldGravity {z=-980}
    simSpaceSettings {maxAngularAcceleration=20, maxAngularVelocity=2, maxLinearAcceleration=100, maxLinearVelocity=20, velocityScaleZ=0, worldAlpha=1}
   }
   n17 = Get r_arm_up() -> (r_arm_up:double)
   n18 = "Transform (Modify) Bone - Bone: dyn_towel_r_00"(ComponentPose=@n16, Alpha=@n17, Translation="-0.239,-0.818,-1.687", Rotation="0.000,5.000,0.000", Scale="1.000,0.500,0.750") {
    rotationMode=BMM_Additive
    rotationSpace=BCS_BoneSpace
    scaleMode=BMM_Additive
    scaleSpace=BCS_BoneSpace
    translationMode=BMM_Additive
    translationSpace=BCS_BoneSpace
    alphaBoolBlend {blendOption=Linear}
    alphaScaleBiasClamp {inRange {max=0, min=-100}, outRange {max=0, min=1}}
    boneToModify {boneName=dyn_towel_r_00}
    rotation {yaw=5}
    scale {y=0.5, z=0.75}
    translation=<-0.2388 -0.8182 -1.687>
   }
   n19 = "Transform (Modify) Bone - Bone: dyn_towel_r_01"(ComponentPose=@n18, Alpha=@n17, Translation="-0.244,-1.965,2.174", Rotation="4.918,-8.140,0.652", Scale="1.000,1.750,1.250") {
    rotationMode=BMM_Additive
    rotationSpace=BCS_BoneSpace
    scaleMode=BMM_Additive
    scaleSpace=BCS_BoneSpace
    translationMode=BMM_Additive
    translationSpace=BCS_BoneSpace
    alphaBoolBlend {blendOption=Linear}
    alphaScaleBiasClamp {inRange {max=0, min=-100}, outRange {max=0, min=1}}
    boneToModify {boneName=dyn_towel_r_01}
    rotation=<rot 4.918 -8.1403 0.652>
    scale {y=1.75, z=1.25}
    translation=<-0.244 -1.965 2.174>
   }
   n20 = Get r_pistonAimUp_alpha() -> (r_pistonAimUp_alpha:double)
   n21 = "Transform (Modify) Bone - Bone: dyn_towel_r_00"(ComponentPose=@n19, Alpha=@n20, Translation="-0.000,-2.315,-0.022", Rotation="0,0,0", Scale="1,1,1") {
    rotationMode=BMM_Additive
    rotationSpace=BCS_BoneSpace
    scaleMode=BMM_Additive
    scaleSpace=BCS_BoneSpace
    translationMode=BMM_Additive
    translationSpace=BCS_BoneSpace
    alphaBoolBlend {blendOption=Linear}
    alphaScaleBiasClamp {inRange {max=0, min=-100}, outRange {max=0, min=1}}
    boneToModify {boneName=dyn_towel_r_00}
    translation {y=-2.315, z=-0.022}
   }
   n22 = "Transform (Modify) Bone - Bone: dyn_towel_r_01"(ComponentPose=@n21, Alpha=@n20, Translation="-0.086,0.135,-0.218", Rotation="-0.059,-10.678,-0.480", Scale="1.000,0.750,0.500") {
    rotationMode=BMM_Additive
    rotationSpace=BCS_BoneSpace
    scaleMode=BMM_Additive
    scaleSpace=BCS_BoneSpace
    translationMode=BMM_Additive
    translationSpace=BCS_BoneSpace
    alphaBoolBlend {blendOption=Linear}
    alphaScaleBiasClamp {inRange {max=0, min=-100}, outRange {max=0, min=1}}
    boneToModify {boneName=dyn_towel_r_01}
    rotation=<rot -0.059 -10.6785 -0.48>
    scale {y=0.75, z=0.5}
    translation=<-0.086 0.135 -0.218>
   }
   n23 = "Transform (Modify) Bone - Bone: dyn_towel_r_02"(ComponentPose=@n22, Alpha=@n20, Translation="0.038,-0.090,0.618", Rotation="-0.234,-5.690,-6.028", Scale="1,1,1") {
    rotationMode=BMM_Additive
    rotationSpace=BCS_BoneSpace
    scaleMode=BMM_Additive
    scaleSpace=BCS_BoneSpace
    translationMode=BMM_Additive
    translationSpace=BCS_BoneSpace
    alphaBoolBlend {blendOption=Linear}
    alphaScaleBiasClamp {inRange {max=0, min=-100}, outRange {max=0, min=1}}
    boneToModify {boneName=dyn_towel_r_02}
    rotation=<rot -0.234 -5.6901 -6.0281>
    translation=<0.038 -0.09 0.618>
   }
   n24 = "Transform (Modify) Bone - Bone: dyn_towel_r_03"(ComponentPose=@n23, Alpha=@n20, Translation="0.003,0.059,-0.565", Rotation="-5.158,-4.993,-0.570", Scale="1,1,1") {
    rotationMode=BMM_Additive
    rotationSpace=BCS_BoneSpace
    scaleMode=BMM_Additive
    scaleSpace=BCS_BoneSpace
    translationMode=BMM_Additive
    translationSpace=BCS_BoneSpace
    alphaBoolBlend {blendOption=Linear}
    alphaScaleBiasClamp {inRange {max=0, min=-100}, outRange {max=0, min=1}}
    boneToModify {boneName=dyn_towel_r_03}
    rotation=<rot -5.158 -4.9931 -0.57>
    translation=<0.003 0.059 -0.565>
   }
   Output Pose(Result=Component To Local(ComponentPose=@n24)) {layerGroup=None}
  ]}
 EventGraph {nodes [
   Event BlueprintUpdateAnimation()
   > n1 = Sequence!(execute) -> (then_0, then_1, then_2:Exec, then_3:Exec, then_4:Exec, then_5:Exec, then_6:Exec)
   n2 = Branch(execute=@n1.then_4, Condition=Get isRifle() -> (isRifle:bool))
   > Set r_pistonAimUp_alpha(r_pistonAimUp_alpha:double=MapRangeClamped(Value=Get r_arm_up(), InRangeA=0, InRangeB=0.5)) -> (Output_Get:double)
   Set r_pistonAimUp_alpha(execute=@n2.else, r_pistonAimUp_alpha=0)
   n3 = BreakTransform(InTransform=GetDeltaTransformFromRefPose(self=GetOwnerSkeletalMeshComponent(), BoneName=thigh_r, BaseName=pelvis))
   Set thigh_rot_r(execute=@n1.then_2, thigh_rot_r=MapRangeClamped(Value=BreakRotator(InRot=@n3.Rotation).Yaw, InRangeA=0, InRangeB=-90))
   n4 = GetPartSkeletalMeshComponent()
   Set l_arm_fwd(
    execute=@n1.then_0
    l_arm_fwd:double=GetAlphaFromReferenceTransformDelta(
     SkeletalMeshComponent=@n4
     BoneName=lowerarm_l
     BaseName=dyn_arm_measure_l
     DeltaComponentType=TranslationY
     InputRange=(X=5,Y=25)
     TwistAxis="1,0,0"
    )
   ) -> (Output_Get:double)
   > Set r_arm_fwd(r_arm_fwd:double=GetAlphaFromReferenceTransformDelta(
     SkeletalMeshComponent=@n4
     BoneName=lowerarm_r
     BaseName=dyn_arm_measure_r
     DeltaComponentType=TranslationY
     InputRange=(X=-25,Y=-5)
     OutputRange=(X=1,Y=0)
     TwistAxis="1,0,0"
    )) -> (Output_Get:double)
   n5 = GetPartSkeletalMeshComponent()
   > Set r_arm_up(r_arm_up:double=GetAlphaFromReferenceTransformDelta(
     SkeletalMeshComponent=@n5
     BoneName=lowerarm_r
     BaseName=dyn_arm_measure_r
     DeltaComponentType=TranslationZ
     InputRange=(X=-25,Y=10)
     OutputRange=(X=1,Y=0)
     TwistAxis="1,0,0"
    )) -> (Output_Get:double)
   > Set l_arm_up(l_arm_up:double=GetAlphaFromReferenceTransformDelta(
     SkeletalMeshComponent=@n5
     BoneName=lowerarm_l
     BaseName=dyn_arm_measure_l
     DeltaComponentType=TranslationZ
     InputRange=(X=10,Y=25)
     TwistAxis="1,0,0"
    )) -> (Output_Get:double)
   > Set rifle_alpha(rifle_alpha:double=GetAlphaFromReferenceTransformDelta(
     SkeletalMeshComponent=@n5
     BoneName=lowerarm_r
     BaseName=spine_04
     DeltaComponentType=TranslationZ
     InputRange=(X=15,Y=20)
     TwistAxis="1,0,0"
    )) -> (Output_Get:double)
   n6 = BreakTransform(InTransform=GetDeltaTransformFromRefPose(self=GetOwnerSkeletalMeshComponent(), BoneName=thigh_l, BaseName=pelvis))
   Set thigh_rot_l(execute=@n1.then_1, thigh_rot_l=MapRangeClamped(Value=BreakRotator(InRot=@n6.Rotation).Yaw, InRangeA=0, InRangeB=-110))
   n7 = Cast To FortPlayerAnimInstance(execute=@n1.then_3, Object=GetAnimInstance(self=GetOwnerSkeletalMeshComponent()))
   > Set isRifle(isRifle:bool=false) -> (Output_Get:bool)
   > n8 = Switch on EFortWeaponCoreAnimation!(execute, Selection:EFortWeaponCoreAnimation Enum=Get WeaponCoreAnim(self=@n7)) -> (
    Melee:Exec
    Pistol:Exec
    Shotgun:Exec
    PaperBlueprint:Exec
    Rifle:Exec
    MeleeOneHand:Exec
    MachinePistol:Exec
    RocketLauncher:Exec
    GrenadeLauncher:Exec
    GoingCommando:Exec
    AssaultRifle:Exec
    TacticalShotgun:Exec
    SniperRifle:Exec
    TrapPlacement:Exec
    ShoulderLauncher:Exec
    AbilityDecoTool:Exec
    Crossbow:Exec
    C4:Exec
    RemoteControl:Exec
    DualWield:Exec
    AR_BullPup:Exec
    AR_ForwardGrip:Exec
    MedPackPaddles:Exec
    SMG_P90:Exec
    AR_DrumGun:Exec
    Consumable_Small:Exec
    Consumable_Large:Exec
    Balloon:Exec
    MountedTurret:Exec
    CreativeTool:Exec
    ExplosiveBow:Exec
    AshtonIndigo:Exec
    AshtonChicago:Exec
    MeleeDualWield:Exec
    Unarmed:Exec
   )
   Branch(execute=[@n8.AssaultRifle, @n8.Rifle, @n8.Shotgun, @n8.SniperRifle, @n8.TacticalShotgun])
   > Set isRifle(isRifle=true)
   Get bIsTargeting()
  ]}
]
variables [
 isRifle {type=boolean}
 l_arm_fwd {type=number}
 l_arm_up {type=number}
 r_arm_fwd {type=number}
 r_arm_up {type=number}
 r_pistonAimUp_alpha {type=number}
 rifle_alpha {type=number}
 thigh_rot_l {type=number}
 thigh_rot_r {type=number}
]
